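# Physics asset for the F_MED_RockerPunkAlt female body, named as its running variant, that sets up dangling hip chains and their buckles.
# Capsules on pelvis and thighs form the body, with small base buckle bodies constrained to the pelvis and carrying swinging buckles and chain segments.
# Front and back long-chain halves join at their lower segments, closing a loop on each side.
# Joints mix tight angular limits with free axes, differing between left and right.
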PhysicsAsset /Game/Characters/Player/Female/Medium/Bodies/F_MED_RockerPunkAlt/Meshes/F_MED_RockerPunkAlt_PhysicsRunning
bodies {
 bace_dyn_longchain_buckle_r [{shapeType=Capsule, length=0.505, radius=0.505, shapeName=bace_dyn_longchain_buckle_r_capsule, rotation=<rot 0 -90 0>}]
 base_dyn_back_buckle_l [{shapeType=Capsule, length=0.505, radius=0.505, shapeName=base_dyn_back_buckle_l_capsule, rotation=<rot 0 -90 0>}]
 base_dyn_back_buckle_r [{shapeType=Capsule, length=0.505, radius=0.505, shapeName=base_dyn_back_buckle_r_capsule, rotation=<rot 0 -90 0>}]
 base_dyn_longchain_buckle_l [{shapeType=Capsule, length=0.505, radius=0.505, shapeName=base_dyn_longchain_buckle_l_capsule, rotation=<rot 0 -90 0>}]
 base_dyn_shortchain_buckle_l [{shapeType=Capsule, length=0.505, radius=0.505, shapeName=base_dyn_shortchain_buckle_l_capsule, rotation=<rot 0 -90 0>}]
 base_dyn_shortchain_buckle_r [{shapeType=Capsule, length=0.505, radius=0.505, shapeName=base_dyn_shortchain_buckle_r_capsule, rotation=<rot 0 -90 0>}]
 dyn_back_buckle_l [{shapeType=Capsule, length=1, radius=1, shapeName=dyn_back_buckle_l_capsule, center=<1.295 -1.3239 -0.2109>, rotation=<rot 0 90 0>}]
 dyn_back_buckle_r [{shapeType=Capsule, length=1, radius=1, shapeName=dyn_back_buckle_r_capsule, center=<1.6461 -0.5671 0.1963>, rotation=<rot 0 90 0>}]
 dyn_longchain_b_l [{
   shapeType=Capsule
   length=8.1686
   radius=1
   shapeName=dyn_longchain_b_l_capsule
   center=<8.4613 0.8498 -0.6432>
   rotation=<rot -0.5711 -77.4497 -85.6238>
  }]
 dyn_longchain_b_lower_l [{shapeType=Capsule, length=8.2199, radius=1, shapeName=dyn_longchain_b_lower_l_capsule, center=<6.9372 0 0>, rotation=<rot 90 0 0>}]
 dyn_longchain_b_lower_r [{
   shapeType=Capsule
   length=8.2199
   radius=1
   shapeName=dyn_longchain_b_lower_r_capsule
   center=<9.437 1.109 -1.1083>
   rotation=<rot 0 -70.3321 -76.7961>
  }]
 dyn_longchain_b_r [{
   shapeType=Capsule
   length=6.4922
   radius=1
   shapeName=dyn_longchain_b_r_capsule
   center=<8.5803 0.5883 -0.6505>
   rotation=<rot -1.4638 -65.5345 -82.9406>
  }]
 dyn_longchain_buckle_l [{shapeType=Capsule, length=1, radius=1, shapeName=dyn_longchain_buckle_l_capsule, center=<1 0 0>}]
 dyn_longchain_f_l [{
   shapeType=Capsule
   length=6.6555
   radius=1.4603
   shapeName=dyn_longchain_f_l_capsule
   center=<7.7979 0.4548 -0.1648>
   rotation=<rot 0 -90.7798 -84.6995>
  }]
 dyn_longchain_f_lower_l [
  {
   shapeType=Capsule
   length=5.5002
   radius=1
   shapeName=dyn_longchain_f_lower_l_capsule
   center=<8.6639 -1.288 1.2081>
   rotation=<rot 0 -112.0374 -93.9347>
  }
  {shapeType=Capsule, length=2, radius=1, shapeName=dyn_longchain_f_lower_l_capsule, center=<12 -9.5 0>, rotation=<rot 90 0 0>}
 ]
 dyn_longchain_f_lower_r [{
   shapeType=Capsule
   length=5.5002
   radius=1
   shapeName=dyn_longchain_f_lower_r_capsule
   center=<8.3927 -1.1782 -0.2864>
   rotation=<rot 0 -112.0377 -87.5765>
  }]
 dyn_longchain_f_r [{
   shapeType=Capsule
   length=6.7226
   radius=1
   shapeName=dyn_longchain_f_r_capsule
   center=<5.5626 -0.1596 0.4474>
   rotation=<rot 0 89.1655 -86.056>
  }]
 dyn_longchain_r_buckle [{shapeType=Capsule, length=1, radius=1, shapeName=dyn_longchain_r_buckle_capsule, center=<1.7702 1 -0.2675>, rotation=<rot 0 90 0>}]
 dyn_shortchain_b_l [{
   shapeType=Capsule
   length=7.5757
   radius=1
   shapeName=dyn_shortchain_b_l_capsule
   center=<8.2623 0.8055 -0.1637>
   rotation=<rot -0.279 -52.9108 -88.982>
  }]
 dyn_shortchain_b_r [{
   shapeType=Capsule
   length=7.5757
   radius=1
   shapeName=dyn_shortchain_b_r_capsule
   center=<8.2623 0.8055 -0.1637>
   rotation=<rot -0.279 -52.9108 -88.982>
  }]
 dyn_shortchain_buckle_l [{shapeType=Capsule, length=1, radius=1, shapeName=dyn_shortchain_buckle_l_capsule, center=<1 1 0>}]
 dyn_shortchain_buckle_r [{shapeType=Capsule, length=1, radius=1, shapeName=dyn_shortchain_buckle_r_capsule, center=<1.6578 1.2722 0.1006>, rotation=<rot 0 90 0>}]
 dyn_shortchain_f_l [{
   shapeType=Capsule
   length=2
   radius=1
   shapeName=dyn_shortchain_f_l_capsule
   center=<2.4475 1.2138 0.127>
   rotation=<rot 8.2215 85.1962 -99.5118>
  }]
 dyn_shortchain_f_r [{
   shapeType=Capsule
   length=2.6967
   radius=1
   shapeName=dyn_shortchain_f_r_capsule
   center=<6.3043 0.2322 0.7963>
   rotation=<rot -0.0084 80.1854 -89.9501>
  }]
 pelvis [{
   shapeType=Capsule
   length=10.1726
   radius=10.2099
   shapeName=pelvis_capsule
   center=<-1.3821 -1.9236 -0.0454>
   rotation=<rot 0.412 -90 0.0355>
  }]
 thigh_l [{
   shapeType=Capsule
   length=22.7643
   radius=8.5501
   shapeName=thigh_l_capsule
   center=<-14.4058 -0.6639 0.4571>
   rotation=<rot -0.4628 -95.0824 84.9571>
  }]
 thigh_r [{
   shapeType=Capsule
   length=22.7643
   radius=8.5501
   shapeName=thigh_r_capsule
   center=<14.5497 1.4144 -0.6953>
   rotation=<rot 0.0247 89.5729 -84.936>
  }]
}
constraints [
 {bone1Name=thigh_l, bone2Name=pelvis, swing1LimitDegrees=45, swing2LimitDegrees=45, twistLimitDegrees=45}
 {bone1Name=thigh_r, bone2Name=pelvis, swing1LimitDegrees=35, swing2LimitDegrees=35, twistLimitDegrees=45}
 {
  bone1Name=bace_dyn_longchain_buckle_r
  bone2Name=pelvis
  swing1LimitDegrees=45
  swing1Motion=Free
  swing2LimitDegrees=45
  swing2Motion=Free
  twistLimitDegrees=45
  twistMotion=Free
 }
 {bone1Name=dyn_longchain_r_buckle, bone2Name=bace_dyn_longchain_buckle_r, swing1LimitDegrees=20, swing2LimitDegrees=5, twistLimitDegrees=1}
 {bone1Name=dyn_longchain_f_r, bone2Name=dyn_longchain_r_buckle, swing1LimitDegrees=49, swing2LimitDegrees=17, twistLimitDegrees=1}
 {bone1Name=dyn_longchain_f_lower_r, bone2Name=dyn_longchain_f_r, swing1LimitDegrees=40, swing2LimitDegrees=40, twistLimitDegrees=1}
 {
  bone1Name=base_dyn_back_buckle_l
  bone2Name=pelvis
  swing1LimitDegrees=45
  swing1Motion=Free
  swing2LimitDegrees=45
  swing2Motion=Free
  twistLimitDegrees=45
  twistMotion=Free
 }
 {bone1Name=dyn_back_buckle_l, bone2Name=base_dyn_back_buckle_l, swing1LimitDegrees=20, swing2LimitDegrees=5, twistLimitDegrees=1}
 {
  bone1Name=dyn_longchain_b_l
  bone2Name=dyn_back_buckle_l
  swing1LimitDegrees=45
  swing2LimitDegrees=45
  twistLimitDegrees=0
  twistMotion=Free
 }
 {
  bone1Name=dyn_longchain_b_lower_l
  bone2Name=dyn_longchain_b_l
  swing1LimitDegrees=40
  swing1Motion=Free
  swing2LimitDegrees=20
  swing2Motion=Free
  twistLimitDegrees=1
  twistMotion=Free
 }
 {bone1Name=dyn_shortchain_b_l, bone2Name=dyn_back_buckle_l, swing1LimitDegrees=25, swing2LimitDegrees=10, twistLimitDegrees=25}
 {
  bone1Name=base_dyn_back_buckle_r
  bone2Name=pelvis
  swing1LimitDegrees=45
  swing1Motion=Free
  swing2LimitDegrees=45
  swing2Motion=Free
  twistLimitDegrees=45
  twistMotion=Free
 }
 {bone1Name=dyn_back_buckle_r, bone2Name=base_dyn_back_buckle_r, swing1LimitDegrees=25, swing2LimitDegrees=5, twistLimitDegrees=1}
 {
  bone1Name=dyn_longchain_b_r
  bone2Name=dyn_back_buckle_r
  swing1LimitDegrees=55
  swing1Motion=Free
  swing2LimitDegrees=44
  swing2Motion=Free
  twistLimitDegrees=1
  twistMotion=Free
 }
 {
  bone1Name=dyn_longchain_b_lower_r
  bone2Name=dyn_longchain_b_r
  swing1LimitDegrees=22.92
  swing1Motion=Free
  swing2LimitDegrees=45
  swing2Motion=Free
  twistLimitDegrees=1
  twistMotion=Free
 }
 {bone1Name=dyn_shortchain_b_r, bone2Name=dyn_back_buckle_r, swing1LimitDegrees=25, swing2LimitDegrees=10, twistLimitDegrees=25}
 {
  bone1Name=base_dyn_longchain_buckle_l
  bone2Name=pelvis
  swing1LimitDegrees=20
  swing1Motion=Locked
  swing2LimitDegrees=5
  swing2Motion=Locked
  twistLimitDegrees=1
  twistMotion=Locked
 }
 {
  bone1Name=dyn_longchain_f_lower_l
  bone2Name=dyn_longchain_f_l
  swing1LimitDegrees=40
  swing1Motion=Free
  swing2LimitDegrees=40
  swing2Motion=Free
  twistLimitDegrees=40
 }
 {bone1Name=base_dyn_shortchain_buckle_l, bone2Name=pelvis, swing1LimitDegrees=20, swing2LimitDegrees=5, twistLimitDegrees=1}
 {bone1Name=dyn_shortchain_f_l, bone2Name=dyn_shortchain_buckle_l, swing1LimitDegrees=5, swing2LimitDegrees=10, twistLimitDegrees=1}
 {bone1Name=base_dyn_shortchain_buckle_r, bone2Name=pelvis, swing1LimitDegrees=20, swing2LimitDegrees=5, twistLimitDegrees=1}
 {bone1Name=dyn_shortchain_buckle_r, bone2Name=base_dyn_shortchain_buckle_r, swing1LimitDegrees=30, swing2LimitDegrees=5, twistLimitDegrees=1}
 {bone1Name=dyn_shortchain_f_r, bone2Name=dyn_shortchain_buckle_r, swing1LimitDegrees=5, swing2LimitDegrees=10, twistLimitDegrees=1}
 {bone1Name=dyn_longchain_b_lower_r, bone2Name=dyn_longchain_f_lower_r, swing1LimitDegrees=40, swing2LimitDegrees=20, twistLimitDegrees=0}
 {
  bone1Name=dyn_longchain_b_lower_l
  bone2Name=dyn_longchain_f_lower_l
  swing1LimitDegrees=40
  swing1Motion=Free
  swing2LimitDegrees=20
  swing2Motion=Free
  twistLimitDegrees=1
  twistMotion=Free
 }
 {bone1Name=dyn_longchain_f_r, bone2Name=dyn_shortchain_f_r, swing1LimitDegrees=5, swing2LimitDegrees=20, twistLimitDegrees=1}
 {bone1Name=dyn_shortchain_buckle_l, bone2Name=base_dyn_shortchain_buckle_l, swing1LimitDegrees=30, swing2LimitDegrees=5, twistLimitDegrees=1}
 {bone1Name=dyn_longchain_buckle_l, bone2Name=base_dyn_longchain_buckle_l, swing1LimitDegrees=20, swing2LimitDegrees=5, twistLimitDegrees=1}
 {
  bone1Name=dyn_longchain_f_l
  bone2Name=dyn_longchain_buckle_l
  swing1LimitDegrees=45
  swing2LimitDegrees=45
  twistLimitDegrees=45
  twistMotion=Free
 }
]
