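# Upper-body physics asset for the medium female RockerPunkAlt player outfit, built for secondary motion on the shirt sleeves.
# Capsules on spine_03 and both upper arms form the core, with free-rotation constraints linking spine, arms and twist bones.
# Four dyn_sleeve bones per arm hang off the twist bones with tight angular limits, mostly favoring one swing axis.
# The right bottom sleeve routes through an extra MASTER bone, and the back shirt detail is rotation-locked to spine_03.
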PhysicsAsset /Game/Characters/Player/Female/Medium/Bodies/F_MED_RockerPunkAlt/Meshes/F_MED_RockerPunkAlt_Physics_UpperBody
bodies {
 MASTER_sleeve_upperbottom_r [{shapeType=Capsule, length=0.505, radius=0.505, shapeName=MASTER_sleeve_upperbottom_r_capsule, rotation=<rot 90 0 0>}]
 dyn_back_shirtdetail [{shapeType=Capsule, length=0.505, radius=0.505, shapeName=dyn_back_shirtdetail_capsule, rotation=<rot 0 -90 0>}]
 dyn_sleeve_sholdertop_l [{shapeType=Capsule, length=1, radius=1, shapeName=dyn_sleeve_sholdertop_l_capsule, center=<11.6392 0 2.9449>, rotation=<rot 90 0 0>}]
 dyn_sleeve_sholdertop_r [{shapeType=Capsule, length=1, radius=1, shapeName=dyn_sleeve_sholdertop_r_capsule, center=<6.7272 3.6086 -1.4082>, rotation=<rot 90 0 0>}]
 dyn_sleeve_upperback_l [{shapeType=Capsule, length=5, radius=1, shapeName=dyn_sleeve_upperback_l_capsule, center=<7.0243 0.1426 1.4161>, rotation=<rot 90 0 0>}]
 dyn_sleeve_upperback_r [{shapeType=Capsule, length=8, radius=1, shapeName=dyn_sleeve_upperback_r_capsule, center=<5.9931 -0.3301 3.6625>, rotation=<rot 90 0 0>}]
 dyn_sleeve_upperbottom_l [{shapeType=Capsule, length=1, radius=1, shapeName=dyn_sleeve_upperbottom_l_capsule, center=<3.961 0 0.2903>, rotation=<rot 90 0 0>}]
 dyn_sleeve_upperbottom_r [{shapeType=Capsule, length=1, radius=1, shapeName=dyn_sleeve_upperbottom_r_capsule, center=<3.5409 0 0>, rotation=<rot 90 0 0>}]
 dyn_sleeve_upperfront_l [{shapeType=Capsule, length=1, radius=1, shapeName=dyn_sleeve_upperfront_l_capsule, center=<5.0959 -0.6076 -0.8152>, rotation=<rot 90 0 0>}]
 dyn_sleeve_upperfront_r [{shapeType=Capsule, length=1, radius=1, shapeName=dyn_sleeve_upperfront_r_capsule, center=<6.3224 0 0>, rotation=<rot 90 0 0>}]
 spine_03 [{
   shapeType=Capsule
   length=6.7597
   radius=2.8558
   shapeName=spine_03_capsule
   center=<3.1931 1.1297 -0.065>
   rotation=<rot 1.0418 -90 -0.1172>
  }]
 upperarm_l [{
   shapeType=Capsule
   length=14.6442
   radius=5.1346
   shapeName=upperarm_l_capsule
   center=<12.6581 -1.4348 -0.8149>
   rotation=<rot -0.0199 -90.1095 -89.8889>
  }]
 upperarm_r [{
   shapeType=Capsule
   length=15.621
   radius=4.4432
   shapeName=upperarm_r_capsule
   center=<-13.7005 0.7756 0.5497>
   rotation=<rot 0.6853 -89.2625 -92.5601>
  }]
 upperarm_twist_01_l [{shapeType=Capsule, length=1, radius=1, shapeName=upperarm_twist_01_l_capsule, rotation=<rot 90 0 0>}]
 upperarm_twist_01_r [{shapeType=Capsule, length=1, radius=1, shapeName=upperarm_twist_01_r_capsule, rotation=<rot 90 0 0>}]
}
constraints [
 {
  bone1Name=upperarm_l
  bone2Name=spine_03
  swing1LimitDegrees=45
  swing1Motion=Free
  swing2LimitDegrees=45
  swing2Motion=Free
  twistLimitDegrees=45
  twistMotion=Free
 }
 {
  bone1Name=upperarm_r
  bone2Name=spine_03
  swing1LimitDegrees=45
  swing1Motion=Free
  swing2LimitDegrees=45
  swing2Motion=Free
  twistLimitDegrees=45
  twistMotion=Free
 }
 {
  bone1Name=dyn_back_shirtdetail
  bone2Name=spine_03
  swing1LimitDegrees=45
  swing1Motion=Locked
  swing2LimitDegrees=45
  swing2Motion=Locked
  twistLimitDegrees=45
  twistMotion=Locked
 }
 {bone1Name=dyn_sleeve_sholdertop_l, bone2Name=upperarm_twist_01_l, swing1LimitDegrees=1, swing2LimitDegrees=15, twistLimitDegrees=1}
 {bone1Name=dyn_sleeve_upperback_l, bone2Name=upperarm_twist_01_l, swing1LimitDegrees=20, swing2LimitDegrees=1, twistLimitDegrees=1}
 {bone1Name=dyn_sleeve_upperbottom_l, bone2Name=upperarm_twist_01_l, swing1LimitDegrees=1, swing2LimitDegrees=10, twistLimitDegrees=1}
 {bone1Name=dyn_sleeve_upperfront_l, bone2Name=upperarm_twist_01_l, swing1LimitDegrees=15, swing2LimitDegrees=1, twistLimitDegrees=1}
 {bone1Name=dyn_sleeve_sholdertop_r, bone2Name=upperarm_twist_01_r, swing1LimitDegrees=1, swing2LimitDegrees=20, twistLimitDegrees=1}
 {bone1Name=dyn_sleeve_upperback_r, bone2Name=upperarm_twist_01_r, swing1LimitDegrees=15, swing2LimitDegrees=1, twistLimitDegrees=1}
 {bone1Name=dyn_sleeve_upperbottom_r, bone2Name=MASTER_sleeve_upperbottom_r, swing1LimitDegrees=5, swing2LimitDegrees=10, twistLimitDegrees=1}
 {bone1Name=dyn_sleeve_upperfront_r, bone2Name=upperarm_twist_01_r, swing1LimitDegrees=20, swing2LimitDegrees=1, twistLimitDegrees=1}
 {bone1Name=MASTER_sleeve_upperbottom_r, bone2Name=upperarm_twist_01_r, swing1LimitDegrees=1, swing2LimitDegrees=1, twistLimitDegrees=1}
 {
  bone1Name=upperarm_twist_01_l
  bone2Name=upperarm_l
  swing1LimitDegrees=45
  swing1Motion=Free
  swing2LimitDegrees=45
  swing2Motion=Free
  twistLimitDegrees=45
  twistMotion=Free
 }
 {
  bone1Name=upperarm_twist_01_r
  bone2Name=upperarm_r
  swing1LimitDegrees=45
  swing1Motion=Free
  swing2LimitDegrees=45
  swing2Motion=Free
  twistLimitDegrees=45
  twistMotion=Free
 }
]
