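# Secondary-motion AnimBP for the female RockerPunkAlt player body, animating its shirt, sleeves, boot parts and hanging chains.
# It copies the attached parent's pose through a Control Rig, then runs crouch-switched and upper-body RigidBody simulations plus per-bone AnimDynamics for shirt and boots.
# Layered bone rotations then correct chains and sleeves per movement or aiming state, mostly additively.
# The EventGraph updates gravity, blend alphas and the crouch switch from locomotion flags and aim pitch.
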AnimBlueprint /Game/Characters/Player/Female/Medium/Bodies/F_MED_RockerPunkAlt/Meshes/F_MED_RockerPunkAlt_AnimBP
target_skeleton=&/F_MED_RockerPunkAlt_Skeleton
graphs [
 AnimGraph {nodes [
   Control Rig(Alpha=0) {
    alphaBoolBlend {blendOption=Linear}
    alphaScaleBiasClamp {bInterpResult=true}
    controlRigAssetReference {blueprintRigClass=&/F_MED_DeformRig_Master_ControlRig_ShoulderFix.F_MED_DeformRig_Master_ControlRig_ShoulderFix_C}
   }
   n1 = Get GravityOverride()
   n2 = RigidBody!(ComponentPose=Local To Component(LocalPose=Use cached pose 'Input node'()), Alpha, OverrideWorldGravity:Vector=@n1, ExternalForce) -> (Pose) {
    bOverrideWorldGravity=true
    overridePhysicsAsset=&/F_MED_RockerPunkAlt_Physics
    simulationSpace=BaseBoneSpace
    alphaBoolBlend {blendOption=Linear}
    baseBoneRef {boneName=pelvis}
    componentAppliedLinearAccClamp=<3000 3000 3000>
    componentLinearAccScale=<0.3 0.3 0.4>
    componentLinearVelScale=<0.3 0.3 0.4>
    simSpaceSettings {maxAngularVelocity=3, maxLinearAcceleration=700, maxLinearVelocity=500, worldAlpha=0.6, externalLinearDragV=<0.2 0.2>}
   }
   n3 = Component To Local(ComponentPose=@n2)
   n4 = RigidBody(ComponentPose=Local To Component(LocalPose=Use cached pose 'Input node'()), OverrideWorldGravity=@n1) {
    bOverrideWorldGravity=true
    overridePhysicsAsset=&/F_MED_RockerPunkAlt_PhysicsRunning
    simulationSpace=BaseBoneSpace
    alphaBoolBlend {blendOption=Linear}
    baseBoneRef {boneName=pelvis}
    componentAppliedLinearAccClamp=<3000 3000 3000>
    componentLinearAccScale=<0.3 0.3 0.4>
    componentLinearVelScale=<0.3 0.3 0.4>
    simSpaceSettings {maxAngularVelocity=3, maxLinearAcceleration=700, maxLinearVelocity=500, worldAlpha=0.6, externalLinearDragV=<0.2 0.2>}
   }
   n5 = Component To Local(ComponentPose=@n4)
   n6 = Get Crouchbody() -> (Crouchbody:bool)
   n7 = RigidBody(
    ComponentPose=Local To Component(LocalPose=Blend Poses by bool(BlendPose_0=@n3, BlendPose_1=@n5, bActiveValue=@n6) {blendType=Linear})
    Alpha=Get sleves_alpha() -> (sleves_alpha:double)
    OverrideWorldGravity=@n1
   ) {
    bOverrideWorldGravity=true
    overridePhysicsAsset=&/F_MED_RockerPunkAlt_Physics_UpperBody
    simulationSpace=BaseBoneSpace
    alphaBoolBlend {blendOption=Linear}
    alphaScaleBiasClamp {bInterpResult=true}
    baseBoneRef {boneName=pelvis}
    componentAppliedLinearAccClamp=<3000 3000 3000>
    componentLinearAccScale=<0.3 0.3 0.4>
    componentLinearVelScale=<0.2 0.2 0.3>
   }
   n8 = Anim Dynamics - Bone: dyn_shirt_f(ComponentPose=@n7, Alpha=1, ExternalForce="0,0,0") {
    angularDampingOverride=0.999
    bGravityOverrideInSimSpace=true
    bOverrideAngularDamping=true
    bUseGravityOverride=true
    alphaBoolBlend {blendOption=Linear}
    boundBone {boneName=dyn_shirt_f}
    componentAppliedLinearAccClamp=<3000 3000 3000>
    componentLinearAccScale=<0.3 0.3 0.3>
    componentLinearVelScale=<0.3 0.3 0.3>
    gravityOverride {z=-980}
    physicsBodyDefinitions [{
      sphereCollisionRadius=0
      boundBone {boneName=dyn_shirt_f}
      boxExtents=<10 12 12>
      constraintSetup {angularLimitsMax=<10 5 30>, angularLimitsMin=<-10 -75 -30>, angularTarget=<0 0 0>}
      localJointOffset=<2 0.2 0.2>
     }]
   }
   n9 = Anim Dynamics - Bone: dyn_shirt_b(ComponentPose=@n8, Alpha=1, ExternalForce="0,0,0") {
    angularDampingOverride=0.999
    bGravityOverrideInSimSpace=true
    bOverrideAngularDamping=true
    bUseGravityOverride=true
    alphaBoolBlend {blendOption=Linear}
    boundBone {boneName=dyn_shirt_b}
    componentAppliedLinearAccClamp=<3000 3000 3000>
    componentLinearAccScale=<0.3 0.3 0.3>
    componentLinearVelScale=<0.3 0.3 0.3>
    gravityOverride {z=-980}
    physicsBodyDefinitions [{
      sphereCollisionRadius=0
      boundBone {boneName=dyn_shirt_b}
      boxExtents=<10 12 12>
      constraintSetup {angularLimitsMax=<10 5 30>, angularLimitsMin=<-10 -75 -30>, angularTarget=<0 0 0>}
      localJointOffset=<2 0.2 0.2>
     }]
   }
   n10 = Anim Dynamics - Bone: dyn_shirt_l(ComponentPose=@n9, Alpha=1, ExternalForce="0,0,0") {
    angularDampingOverride=0.999
    bGravityOverrideInSimSpace=true
    bOverrideAngularDamping=true
    bUseGravityOverride=true
    alphaBoolBlend {blendOption=Linear}
    boundBone {boneName=dyn_shirt_l}
    componentAppliedLinearAccClamp=<3000 3000 3000>
    componentLinearAccScale=<0.3 0.3 0.3>
    componentLinearVelScale=<0.3 0.3 0.3>
    gravityOverride {z=-980}
    physicsBodyDefinitions [{
      sphereCollisionRadius=0
      boundBone {boneName=dyn_shirt_l}
      boxExtents=<10 12 12>
      constraintSetup {angularLimitsMax=<10 5 30>, angularLimitsMin=<-10 -75 -30>, angularTarget=<0 0 0>}
      localJointOffset=<2 0.2 0.2>
     }]
   }
   n11 = Anim Dynamics - Bone: dyn_shirt_r(ComponentPose=@n10, Alpha=1, ExternalForce="0,0,0") {
    angularDampingOverride=0.999
    bGravityOverrideInSimSpace=true
    bOverrideAngularDamping=true
    bUseGravityOverride=true
    alphaBoolBlend {blendOption=Linear}
    boundBone {boneName=dyn_shirt_r}
    componentAppliedLinearAccClamp=<3000 3000 3000>
    componentLinearAccScale=<0.3 0.3 0.3>
    componentLinearVelScale=<0.3 0.3 0.3>
    gravityOverride {z=-980}
    physicsBodyDefinitions [{
      sphereCollisionRadius=0
      boundBone {boneName=dyn_shirt_r}
      boxExtents=<10 12 12>
      constraintSetup {angularLimitsMax=<10 5 30>, angularLimitsMin=<-10 -75 -30>, angularTarget=<0 0 0>}
      localJointOffset=<2 0.2 0.2>
     }]
   }
   n12 = Anim Dynamics - Bone: dyn_rightboot_lace_l(ComponentPose=@n11, Alpha=1, ExternalForce="0,0,0") {
    angularDampingOverride=0.999
    bGravityOverrideInSimSpace=true
    bOverrideAngularDamping=true
    bUseGravityOverride=true
    alphaBoolBlend {blendOption=Linear}
    boundBone {boneName=dyn_rightboot_lace_l}
    componentAppliedLinearAccClamp=<3000 3000 3000>
    componentLinearAccScale=<0.3 0.3 0.3>
    componentLinearVelScale=<0.3 0.3 0.3>
    gravityOverride {z=-980}
    physicsBodyDefinitions [{
      sphereCollisionRadius=0
      boundBone {boneName=dyn_rightboot_lace_l}
      boxExtents=<10 12 12>
      constraintSetup {angularLimitsMax=<10 5 20>, angularLimitsMin=<-10 -75 -10>, angularTarget=<0 0 0>}
      localJointOffset=<2 0.2 0.2>
     }]
   }
   n13 = Anim Dynamics - Bone: dyn_rightboot_r(ComponentPose=@n12, Alpha=1, ExternalForce="0,0,0") {
    angularDampingOverride=0.999
    bGravityOverrideInSimSpace=true
    bOverrideAngularDamping=true
    bUseGravityOverride=true
    alphaBoolBlend {blendOption=Linear}
    boundBone {boneName=dyn_rightboot_r}
    componentAppliedLinearAccClamp=<3000 3000 3000>
    componentLinearAccScale=<0.3 0.3 0.3>
    componentLinearVelScale=<0.3 0.3 0.3>
    gravityOverride {z=-980}
    physicsBodyDefinitions [{
      sphereCollisionRadius=0
      boundBone {boneName=dyn_rightboot_r}
      boxExtents=<10 12 12>
      constraintSetup {angularLimitsMax=<10 5 30>, angularLimitsMin=<-10 -75 -10>, angularTarget=<0 0 0>}
      localJointOffset=<2 0.2 0.2>
     }]
   }
   n14 = Anim Dynamics - Bone: dyn_rightboot_l(ComponentPose=@n13, Alpha=1, ExternalForce="0,0,0") {
    angularDampingOverride=0.999
    bGravityOverrideInSimSpace=true
    bOverrideAngularDamping=true
    bUseGravityOverride=true
    alphaBoolBlend {blendOption=Linear}
    boundBone {boneName=dyn_rightboot_l}
    componentAppliedLinearAccClamp=<3000 3000 3000>
    componentLinearAccScale=<0.3 0.3 0.3>
    componentLinearVelScale=<0.3 0.3 0.3>
    gravityOverride {z=-980}
    physicsBodyDefinitions [{
      sphereCollisionRadius=0
      boundBone {boneName=dyn_rightboot_l}
      boxExtents=<10 12 12>
      constraintSetup {angularLimitsMax=<10 5 30>, angularLimitsMin=<-10 -75 -10>, angularTarget=<0 0 0>}
      localJointOffset=<2 0.2 0.2>
     }]
   }
   n15 = Anim Dynamics - Bone: dyn_rightboot_lace_r(ComponentPose=@n14, Alpha=1, ExternalForce="0,0,0") {
    angularDampingOverride=0.999
    bGravityOverrideInSimSpace=true
    bOverrideAngularDamping=true
    bUseGravityOverride=true
    alphaBoolBlend {blendOption=Linear}
    boundBone {boneName=dyn_rightboot_lace_r}
    componentAppliedLinearAccClamp=<3000 3000 3000>
    componentLinearAccScale=<0.3 0.3 0.3>
    componentLinearVelScale=<0.3 0.3 0.3>
    gravityOverride {z=-980}
    physicsBodyDefinitions [{
      sphereCollisionRadius=0
      boundBone {boneName=dyn_rightboot_lace_r}
      boxExtents=<10 12 12>
      constraintSetup {angularLimitsMax=<10 5 20>, angularLimitsMin=<-10 -75 -10>, angularTarget=<0 0 0>}
      localJointOffset=<2 0.2 0.2>
     }]
   }
   n16 = Anim Dynamics - Bone: dyn_rightboot_lacebottom_l(ComponentPose=@n15, Alpha=1, ExternalForce="0,0,0") {
    angularDampingOverride=0.999
    bGravityOverrideInSimSpace=true
    bOverrideAngularDamping=true
    bUseGravityOverride=true
    alphaBoolBlend {blendOption=Linear}
    boundBone {boneName=dyn_rightboot_lacebottom_l}
    componentAppliedLinearAccClamp=<3000 3000 3000>
    componentLinearAccScale=<0.3 0.3 0.3>
    componentLinearVelScale=<0.3 0.3 0.3>
    gravityOverride {z=-980}
    physicsBodyDefinitions [{
      sphereCollisionRadius=0
      boundBone {boneName=dyn_rightboot_lacebottom_l}
      boxExtents=<10 12 12>
      constraintSetup {angularLimitsMax=<10 5 30>, angularLimitsMin=<-10 -75 -35>, angularTarget=<0 0 0>}
      localJointOffset=<2 0.2 0.2>
     }]
   }
   n17 = Anim Dynamics - Bone: dyn_rightboot_bottom_r(ComponentPose=@n16, Alpha=1, ExternalForce="0,0,0") {
    angularDampingOverride=0.999
    bGravityOverrideInSimSpace=true
    bOverrideAngularDamping=true
    bUseGravityOverride=true
    alphaBoolBlend {blendOption=Linear}
    boundBone {boneName=dyn_rightboot_bottom_r}
    componentAppliedLinearAccClamp=<3000 3000 3000>
    componentLinearAccScale=<0.3 0.3 0.3>
    componentLinearVelScale=<0.3 0.3 0.3>
    gravityOverride {z=-980}
    physicsBodyDefinitions [{
      sphereCollisionRadius=0
      boundBone {boneName=dyn_rightboot_bottom_r}
      boxExtents=<10 12 12>
      constraintSetup {angularLimitsMax=<20 20 30>, angularLimitsMin=<0 0 -15>, angularTarget=<0 0 0>}
      localJointOffset=<2 0.2 0.2>
     }]
   }
   n18 = Anim Dynamics - Bone: dyn_rightboot_bottom_l(ComponentPose=@n17, Alpha=1, ExternalForce="0,0,0") {
    angularDampingOverride=0.999
    bGravityOverrideInSimSpace=true
    bOverrideAngularDamping=true
    bUseGravityOverride=true
    alphaBoolBlend {blendOption=Linear}
    boundBone {boneName=dyn_rightboot_bottom_l}
    componentAppliedLinearAccClamp=<3000 3000 3000>
    componentLinearAccScale=<0.3 0.3 0.3>
    componentLinearVelScale=<0.3 0.3 0.3>
    gravityOverride {z=-980}
    physicsBodyDefinitions [{
      sphereCollisionRadius=0
      boundBone {boneName=dyn_rightboot_bottom_l}
      boxExtents=<10 12 12>
      constraintSetup {angularLimitsMax=<30 30 30>, angularLimitsMin=<0 0 -35>, angularTarget=<0 0 0>}
      localJointOffset=<2 0.2 0.2>
     }]
   }
   n19 = Anim Dynamics - Bone: Dyn_rightboot_lacebottom_r(ComponentPose=@n18, Alpha=1, ExternalForce="0,0,0") {
    angularDampingOverride=0.999
    bGravityOverrideInSimSpace=true
    bOverrideAngularDamping=true
    bUseGravityOverride=true
    alphaBoolBlend {blendOption=Linear}
    boundBone {boneName=Dyn_rightboot_lacebottom_r}
    componentAppliedLinearAccClamp=<3000 3000 3000>
    componentLinearAccScale=<0.3 0.3 0.3>
    componentLinearVelScale=<0.3 0.3 0.3>
    gravityOverride {z=-980}
    physicsBodyDefinitions [{
      sphereCollisionRadius=0
      boundBone {boneName=Dyn_rightboot_lacebottom_r}
      boxExtents=<10 12 12>
      constraintSetup {angularLimitsMax=<10 5 30>, angularLimitsMin=<-10 -75 -35>, angularTarget=<0 0 0>}
      localJointOffset=<2 0.2 0.2>
     }]
   }
   n20 = Anim Dynamics - Bone: dyn_leftboot_lace_l(ComponentPose=@n19, Alpha=1, ExternalForce="0,0,0") {
    angularDampingOverride=0.999
    bGravityOverrideInSimSpace=true
    bOverrideAngularDamping=true
    bUseGravityOverride=true
    alphaBoolBlend {blendOption=Linear}
    boundBone {boneName=dyn_leftboot_lace_l}
    componentAppliedLinearAccClamp=<3000 3000 3000>
    componentLinearAccScale=<0.3 0.3 0.3>
    componentLinearVelScale=<0.3 0.3 0.3>
    gravityOverride {z=-980}
    physicsBodyDefinitions [{
      sphereCollisionRadius=0
      boundBone {boneName=dyn_leftboot_lace_l}
      boxExtents=<10 12 12>
      constraintSetup {angularLimitsMax=<10 5 20>, angularLimitsMin=<-5 -5 -10>, angularTarget=<0 0 0>}
      localJointOffset=<2 0.2 0.2>
     }]
   }
   n21 = Anim Dynamics - Bone: dyn_lefttboot_r(ComponentPose=@n20, Alpha=1, ExternalForce="0,0,0") {
    angularDampingOverride=0.999
    bGravityOverrideInSimSpace=true
    bOverrideAngularDamping=true
    bUseGravityOverride=true
    alphaBoolBlend {blendOption=Linear}
    boundBone {boneName=dyn_lefttboot_r}
    componentAppliedLinearAccClamp=<3000 3000 3000>
    componentLinearAccScale=<0.3 0.3 0.3>
    componentLinearVelScale=<0.3 0.3 0.3>
    gravityOverride {z=-980}
    physicsBodyDefinitions [{
      sphereCollisionRadius=0
      boundBone {boneName=dyn_lefttboot_r}
      boxExtents=<10 12 12>
      constraintSetup {angularLimitsMax=<10 5 30>, angularLimitsMin=<0 0 0>, angularTarget=<0 0 0>}
      localJointOffset=<2 0.2 0.2>
     }]
   }
   n22 = Anim Dynamics - Bone: dyn_lefttboot_l(ComponentPose=@n21, Alpha=1, ExternalForce="0,0,0") {
    angularDampingOverride=0.999
    bGravityOverrideInSimSpace=true
    bOverrideAngularDamping=true
    bUseGravityOverride=true
    alphaBoolBlend {blendOption=Linear}
    boundBone {boneName=dyn_lefttboot_l}
    componentAppliedLinearAccClamp=<3000 3000 3000>
    componentLinearAccScale=<0.3 0.3 0.3>
    componentLinearVelScale=<0.3 0.3 0.3>
    gravityOverride {z=-980}
    physicsBodyDefinitions [{
      sphereCollisionRadius=0
      boundBone {boneName=dyn_lefttboot_l}
      boxExtents=<10 12 12>
      constraintSetup {angularLimitsMax=<10 5 30>, angularLimitsMin=<0 0 0>, angularTarget=<0 0 0>}
      localJointOffset=<2 0.2 0.2>
     }]
   }
   n23 = Anim Dynamics - Bone: dyn_lefttboot_lace_r(ComponentPose=@n22, Alpha=1, ExternalForce="0,0,0") {
    angularDampingOverride=0.999
    bGravityOverrideInSimSpace=true
    bOverrideAngularDamping=true
    bUseGravityOverride=true
    alphaBoolBlend {blendOption=Linear}
    boundBone {boneName=dyn_lefttboot_lace_r}
    componentAppliedLinearAccClamp=<3000 3000 3000>
    componentLinearAccScale=<0.3 0.3 0.3>
    componentLinearVelScale=<0.3 0.3 0.3>
    gravityOverride {z=-980}
    physicsBodyDefinitions [{
      sphereCollisionRadius=0
      boundBone {boneName=dyn_lefttboot_lace_r}
      boxExtents=<10 12 12>
      constraintSetup {angularLimitsMax=<40 5 20>, angularLimitsMin=<-5 -5 -10>, angularTarget=<0 0 0>}
      localJointOffset=<2 0.2 0.2>
     }]
   }
   n24 = Anim Dynamics - Bone: dyn_leftboot_lacebottom_l(ComponentPose=@n23, Alpha=1, ExternalForce="0,0,0") {
    angularDampingOverride=0.999
    bGravityOverrideInSimSpace=true
    bOverrideAngularDamping=true
    bUseGravityOverride=true
    alphaBoolBlend {blendOption=Linear}
    boundBone {boneName=dyn_leftboot_lacebottom_l}
    componentAppliedLinearAccClamp=<3000 3000 3000>
    componentLinearAccScale=<0.3 0.3 0.3>
    componentLinearVelScale=<0.3 0.3 0.3>
    gravityOverride {z=-980}
    physicsBodyDefinitions [{
      sphereCollisionRadius=0
      boundBone {boneName=dyn_leftboot_lacebottom_l}
      boxExtents=<10 12 12>
      constraintSetup {angularLimitsMax=<10 5 30>, angularLimitsMin=<-10 -75 -80>, angularTarget=<0 0 0>}
      localJointOffset=<2 0.2 0.2>
     }]
   }
   n25 = Anim Dynamics - Bone: dyn_lefttboot_bottom_r(ComponentPose=@n24, Alpha=1, ExternalForce="0,0,0") {
    angularDampingOverride=0.999
    bGravityOverrideInSimSpace=true
    bOverrideAngularDamping=true
    bUseGravityOverride=true
    alphaBoolBlend {blendOption=Linear}
    boundBone {boneName=dyn_lefttboot_bottom_r}
    componentAppliedLinearAccClamp=<3000 3000 3000>
    componentLinearAccScale=<0.3 0.3 0.3>
    componentLinearVelScale=<0.3 0.3 0.3>
    gravityOverride {z=-980}
    physicsBodyDefinitions [{
      sphereCollisionRadius=0
      boundBone {boneName=dyn_lefttboot_bottom_r}
      boxExtents=<10 12 12>
      constraintSetup {angularLimitsMax=<20 20 30>, angularLimitsMin=<0 0 -35>, angularTarget=<0 0 0>}
      localJointOffset=<2 0.2 0.2>
     }]
   }
   n26 = Anim Dynamics - Bone: dyn_lefttboot_bottom_l(ComponentPose=@n25, Alpha=1, ExternalForce="0,0,0") {
    angularDampingOverride=0.999
    bGravityOverrideInSimSpace=true
    bOverrideAngularDamping=true
    bUseGravityOverride=true
    alphaBoolBlend {blendOption=Linear}
    boundBone {boneName=dyn_lefttboot_bottom_l}
    componentAppliedLinearAccClamp=<3000 3000 3000>
    componentLinearAccScale=<0.3 0.3 0.3>
    componentLinearVelScale=<0.3 0.3 0.3>
    gravityOverride {z=-980}
    physicsBodyDefinitions [{
      sphereCollisionRadius=0
      boundBone {boneName=dyn_lefttboot_bottom_l}
      boxExtents=<10 12 12>
      constraintSetup {angularLimitsMax=<30 30 30>, angularLimitsMin=<0 0 -35>, angularTarget=<0 0 0>}
      localJointOffset=<2 0.2 0.2>
     }]
   }
   n27 = Anim Dynamics - Bone: dyn_lefttboot_lacebottom_r(ComponentPose=@n26, Alpha=1, ExternalForce="0,0,0") {
    angularDampingOverride=0.999
    bGravityOverrideInSimSpace=true
    bOverrideAngularDamping=true
    bUseGravityOverride=true
    alphaBoolBlend {blendOption=Linear}
    boundBone {boneName=dyn_lefttboot_lacebottom_r}
    componentAppliedLinearAccClamp=<3000 3000 3000>
    componentLinearAccScale=<0.3 0.3 0.3>
    componentLinearVelScale=<0.3 0.3 0.3>
    gravityOverride {z=-980}
    physicsBodyDefinitions [{
      sphereCollisionRadius=0
      boundBone {boneName=dyn_lefttboot_lacebottom_r}
      boxExtents=<10 12 12>
      constraintSetup {angularLimitsMax=<10 5 30>, angularLimitsMin=<-10 -75 -80>, angularTarget=<0 0 0>}
      localJointOffset=<2 0.2 0.2>
     }]
   }
   n28 = Get Jump Fix() -> (Jump Fix:double)
   n29 = "Transform (Modify) Bone - Bone: dyn_shortchain_buckle_r"(ComponentPose=@n27, Alpha=@n28, Translation="0,0,0", Rotation="-3.093,19.314,0.375", Scale="1,1,1") {
    rotationMode=BMM_Additive
    rotationSpace=BCS_BoneSpace
    alphaBoolBlend {blendOption=Linear}
    alphaScaleBiasClamp {bInterpResult=true, interpSpeedDecreasing=5, interpSpeedIncreasing=5}
    boneToModify {boneName=dyn_shortchain_buckle_r}
    rotation=<rot -3.093 19.314 0.3751>
   }
   n30 = "Transform (Modify) Bone - Bone: dyn_shortchain_f_r"(ComponentPose=@n29, Alpha=@n28, Translation="0,0,0", Rotation="-27.608,-28.771,29.793", Scale="1,1,1") {
    rotationMode=BMM_Additive
    rotationSpace=BCS_BoneSpace
    alphaBoolBlend {blendOption=Linear}
    alphaScaleBiasClamp {bInterpResult=true, interpSpeedDecreasing=5, interpSpeedIncreasing=5}
    boneToModify {boneName=dyn_shortchain_f_r}
    rotation=<rot -27.608 -28.771 29.793>
   }
   n31 = "Transform (Modify) Bone - Bone: dyn_longchain_r_buckle"(ComponentPose=@n30, Alpha=@n28, Translation="0,0,0", Rotation="-1.161,60.858,10.762", Scale="1,1,1") {
    rotationMode=BMM_Additive
    rotationSpace=BCS_BoneSpace
    alphaBoolBlend {blendOption=Linear}
    alphaScaleBiasClamp {bInterpResult=true, interpSpeedDecreasing=5, interpSpeedIncreasing=5}
    boneToModify {boneName=dyn_longchain_r_buckle}
    rotation=<rot -1.161 60.858 10.7621>
   }
   n32 = "Transform (Modify) Bone - Bone: dyn_longchain_f_r"(ComponentPose=@n31, Alpha=@n28, Translation="0,0,0", Rotation="30.747,-38.427,-54.950", Scale="1,1,1") {
    rotationMode=BMM_Additive
    rotationSpace=BCS_BoneSpace
    alphaBoolBlend {blendOption=Linear}
    alphaScaleBiasClamp {bInterpResult=true, interpSpeedDecreasing=5, interpSpeedIncreasing=5}
    boneToModify {boneName=dyn_longchain_f_r}
    rotation=<rot 30.747 -38.427 -54.9499>
   }
   n33 = "Transform (Modify) Bone - Bone: dyn_longchain_buckle_l"(ComponentPose=@n32, Alpha=@n28, Translation="0,0,0", Rotation="0.241,-0.052,0.014", Scale="1,1,1") {
    rotationMode=BMM_Additive
    rotationSpace=BCS_BoneSpace
    alphaBoolBlend {blendOption=Linear}
    alphaScaleBiasClamp {bInterpResult=true, interpSpeedDecreasing=5, interpSpeedIncreasing=5}
    boneToModify {boneName=dyn_longchain_buckle_l}
    rotation=<rot 0.241 -0.052 0.014>
   }
   n34 = "Transform (Modify) Bone - Bone: dyn_shortchain_f_l"(ComponentPose=@n33, Alpha=@n28, Translation="0,0,0", Rotation="35.384,-44.230,29.936", Scale="1,1,1") {
    rotationMode=BMM_Additive
    rotationSpace=BCS_BoneSpace
    alphaBoolBlend {blendOption=Linear}
    alphaScaleBiasClamp {bInterpResult=true, interpSpeedDecreasing=5, interpSpeedIncreasing=5}
    boneToModify {boneName=dyn_shortchain_f_l}
    rotation=<rot 35.384 -44.23 29.936>
   }
   n35 = "Transform (Modify) Bone - Bone: dyn_longchain_f_l"(ComponentPose=@n34, Alpha=@n28, Translation="0,0,0", Rotation="0.665,2.589,25.728", Scale="1,1,1") {
    rotationMode=BMM_Additive
    rotationSpace=BCS_BoneSpace
    alphaBoolBlend {blendOption=Linear}
    alphaScaleBiasClamp {bInterpResult=true, interpSpeedDecreasing=5, interpSpeedIncreasing=5}
    boneToModify {boneName=dyn_longchain_f_l}
    rotation=<rot 0.665 2.589 25.728>
   }
   n36 = Get bIsSkyDiving()
   n37 = "Transform (Modify) Bone - Bone: dyn_longchain_f_r"!(ComponentPose=@n35, bAlphaBoolEnabled:bool=@n36, Translation="0,0,0", Rotation="-10,0,0", Scale="1,1,1") -> (Pose) {
    alphaInputType=Bool
    rotationMode=BMM_Additive
    rotationSpace=BCS_BoneSpace
    alphaBoolBlend {blendOption=Linear}
    boneToModify {boneName=dyn_longchain_f_r}
    rotation {pitch=-10}
   }
   n38 = "Transform (Modify) Bone - Bone: dyn_longchain_b_r"!(ComponentPose=@n37, bAlphaBoolEnabled:bool=@n36, Translation="0,0,0", Rotation="-10,0,0", Scale="1,1,1") -> (Pose) {
    alphaInputType=Bool
    rotationMode=BMM_Additive
    rotationSpace=BCS_BoneSpace
    alphaBoolBlend {blendOption=Linear}
    boneToModify {boneName=dyn_longchain_b_r}
    rotation {pitch=-10}
   }
   n39 = "Transform (Modify) Bone - Bone: dyn_longchain_f_l"!(ComponentPose=@n38, bAlphaBoolEnabled:bool=@n36, Translation="0,0,0", Rotation="8.818,-1.639,-19.610", Scale="1,1,1") -> (Pose) {
    alphaInputType=Bool
    rotationMode=BMM_Additive
    rotationSpace=BCS_BoneSpace
    alphaBoolBlend {blendOption=Linear}
    boneToModify {boneName=dyn_longchain_f_l}
    rotation=<rot 8.818 -1.639 -19.61>
   }
   n40 = "Transform (Modify) Bone - Bone: dyn_longchain_b_l"!(ComponentPose=@n39, bAlphaBoolEnabled:bool=@n36, Translation="0,0,0", Rotation="15.248,-1.709,36.183", Scale="1,1,1") -> (Pose) {
    alphaInputType=Bool
    rotationMode=BMM_Additive
    rotationSpace=BCS_BoneSpace
    alphaBoolBlend {blendOption=Linear}
    boneToModify {boneName=dyn_longchain_b_l}
    rotation=<rot 15.248 -1.709 36.183>
   }
   n41 = Get CrouchStill fix() -> (CrouchStill fix:double)
   n42 = "Transform (Modify) Bone - Bone: dyn_shortchain_f_l"(ComponentPose=@n40, Alpha=@n41, Translation="0,0,0", Rotation="66.064,-24.642,-1.146", Scale="1,1,1") {
    rotationMode=BMM_Additive
    rotationSpace=BCS_BoneSpace
    alphaBoolBlend {blendOption=Linear}
    alphaScaleBiasClamp {bInterpResult=true}
    boneToModify {boneName=dyn_shortchain_f_l}
    rotation=<rot 66.064 -24.6421 -1.1461>
   }
   n43 = "Transform (Modify) Bone - Bone: dyn_longchain_f_l"!(ComponentPose=@n42, Alpha:float=@n41, Translation="0,0,0", Rotation="-9.081,-35.372,-87.819", Scale="1,1,1") -> (Pose) {
    rotationMode=BMM_Replace
    rotationSpace=BCS_BoneSpace
    alphaBoolBlend {blendOption=Linear}
    alphaScaleBiasClamp {bInterpResult=true}
    boneToModify {boneName=dyn_longchain_f_l}
    rotation=<rot -9.081 -35.372 -87.819>
   }
   n44 = "Transform (Modify) Bone - Bone: dyn_longchain_b_l"!(ComponentPose=@n43, Alpha:float=@n41, Translation="0,0,0", Rotation="3.288,-5.397,5.503", Scale="1,1,1") -> (Pose) {
    rotationMode=BMM_Additive
    rotationSpace=BCS_BoneSpace
    alphaBoolBlend {blendOption=Linear}
    alphaScaleBiasClamp {bInterpResult=true}
    boneToModify {boneName=dyn_longchain_b_l}
    rotation=<rot 3.288 -5.397 5.503>
   }
   n45 = "Transform (Modify) Bone - Bone: dyn_longchain_f_lower_l"(ComponentPose=@n44, Alpha=@n41, Translation="0,0,0", Rotation="-35.143,0.016,-16.334", Scale="1,1,1") {
    rotationMode=BMM_Additive
    rotationSpace=BCS_BoneSpace
    alphaBoolBlend {blendOption=Linear}
    alphaScaleBiasClamp {bInterpResult=true}
    boneToModify {boneName=dyn_longchain_f_lower_l}
    rotation=<rot -35.143 0.016 -16.334>
   }
   n46 = Get CrouchMove fix() -> (CrouchMove fix:double)
   n47 = "Transform (Modify) Bone - Bone: dyn_longchain_r_buckle"(ComponentPose=@n45, Alpha=@n46, Translation="0,0,0", Rotation="-1.035,33.285,12.153", Scale="1,1,1") {
    rotationMode=BMM_Additive
    rotationSpace=BCS_BoneSpace
    alphaBoolBlend {blendOption=Linear}
    alphaScaleBiasClamp {bInterpResult=true}
    boneToModify {boneName=dyn_longchain_r_buckle}
    rotation=<rot -1.035 33.285 12.153>
   }
   n48 = "Transform (Modify) Bone - Bone: dyn_longchain_f_r"!(ComponentPose=@n47, Alpha:float=@n46, Translation="0,0,0", Rotation="25.761,-25.243,-13.625", Scale="1,1,1") -> (Pose) {
    rotationMode=BMM_Additive
    rotationSpace=BCS_BoneSpace
    alphaBoolBlend {blendOption=Linear}
    alphaScaleBiasClamp {bInterpResult=true}
    boneToModify {boneName=dyn_longchain_f_r}
    rotation=<rot 25.761 -25.243 -13.625>
   }
   n49 = "Transform (Modify) Bone - Bone: dyn_longchain_f_l"(ComponentPose=@n48, Alpha=@n46, Translation="0,0,0", Rotation="13.197,-40.057,-96.644", Scale="1,1,1") {
    rotationMode=BMM_Additive
    rotationSpace=BCS_BoneSpace
    alphaBoolBlend {blendOption=Linear}
    alphaScaleBiasClamp {bInterpResult=true, interpSpeedDecreasing=5, interpSpeedIncreasing=5}
    boneToModify {boneName=dyn_longchain_f_l}
    rotation=<rot 13.197 -40.057 -96.644>
   }
   n50 = "Transform (Modify) Bone - Bone: dyn_shortchain_buckle_r"(ComponentPose=@n49, Alpha=@n46, Translation="0,0,0", Rotation="-28.715,27.475,23.317", Scale="1,1,1") {
    rotationMode=BMM_Additive
    rotationSpace=BCS_BoneSpace
    alphaBoolBlend {blendOption=Linear}
    alphaScaleBiasClamp {bInterpResult=true}
    boneToModify {boneName=dyn_shortchain_buckle_r}
    rotation=<rot -28.715 27.475 23.317>
   }
   n51 = "Transform (Modify) Bone - Bone: dyn_shortchain_f_r"(ComponentPose=@n50, Alpha=@n46, Translation="0,0,0", Rotation="-30.628,-1.535,-5.315", Scale="1,1,1") {
    rotationMode=BMM_Additive
    rotationSpace=BCS_BoneSpace
    alphaBoolBlend {blendOption=Linear}
    alphaScaleBiasClamp {bInterpResult=true}
    boneToModify {boneName=dyn_shortchain_f_r}
    rotation=<rot -30.628 -1.535 -5.315>
   }
   n52 = "Transform (Modify) Bone - Bone: dyn_shortchain_f_l"(ComponentPose=@n51, Alpha=@n46, Translation="0,0,0", Rotation="45.111,-13.914,60.234", Scale="1,1,1") {
    rotationMode=BMM_Additive
    rotationSpace=BCS_BoneSpace
    alphaBoolBlend {blendOption=Linear}
    alphaScaleBiasClamp {bInterpResult=true}
    boneToModify {boneName=dyn_shortchain_f_l}
    rotation=<rot 45.111 -13.914 60.234>
   }
   n53 = "Transform (Modify) Bone - Bone: dyn_longchain_b_l"(ComponentPose=@n52, Alpha=@n46, Translation="0,0,0", Rotation="-15.767,0.684,-30.461", Scale="1,1,1") {
    rotationMode=BMM_Replace
    rotationSpace=BCS_BoneSpace
    alphaBoolBlend {blendOption=Linear}
    alphaScaleBiasClamp {bInterpResult=true, interpSpeedDecreasing=5, interpSpeedIncreasing=5}
    boneToModify {boneName=dyn_longchain_b_l}
    rotation=<rot -15.767 0.684 -30.461>
   }
   n54 = Get SlideFix() -> (SlideFix:double)
   n55 = "Transform (Modify) Bone - Bone: dyn_shortchain_f_r"(ComponentPose=@n53, Alpha=@n54, Translation="0,0,0", Rotation="-27.433,2.608,-12.334", Scale="1,1,1") {
    rotationMode=BMM_Additive
    rotationSpace=BCS_BoneSpace
    alphaBoolBlend {blendOption=Linear}
    alphaScaleBiasClamp {bInterpResult=true}
    boneToModify {boneName=dyn_shortchain_f_r}
    rotation=<rot -27.433 2.608 -12.334>
   }
   n56 = "Transform (Modify) Bone - Bone: dyn_longchain_f_r"(ComponentPose=@n55, Alpha=@n54, Translation="0,0,0", Rotation="-28.043,-0.282,8.513", Scale="1,1,1") {
    rotationMode=BMM_Additive
    rotationSpace=BCS_BoneSpace
    alphaBoolBlend {blendOption=Linear}
    alphaScaleBiasClamp {bInterpResult=true}
    boneToModify {boneName=dyn_longchain_f_r}
    rotation=<rot -28.043 -0.282 8.513>
   }
   n57 = "Transform (Modify) Bone - Bone: dyn_longchain_f_l"(ComponentPose=@n56, Alpha=@n54, Translation="0,0,0", Rotation="-43.821,-28.080,53.591", Scale="1,1,1") {
    rotationMode=BMM_Additive
    rotationSpace=BCS_BoneSpace
    alphaBoolBlend {blendOption=Linear}
    alphaScaleBiasClamp {bInterpResult=true}
    boneToModify {boneName=dyn_longchain_f_l}
    rotation=<rot -43.821 -28.08 53.591>
   }
   n58 = "Transform (Modify) Bone - Bone: dyn_longchain_b_l"(ComponentPose=@n57, Alpha=@n54, Translation="0,0,0", Rotation="12.065,48.891,-8.276", Scale="1,1,1") {
    rotationMode=BMM_Additive
    rotationSpace=BCS_BoneSpace
    alphaBoolBlend {blendOption=Linear}
    alphaScaleBiasClamp {bInterpResult=true}
    boneToModify {boneName=dyn_longchain_b_l}
    rotation=<rot 12.065 48.891 -8.276>
   }
   n59 = "Transform (Modify) Bone - Bone: dyn_shortchain_f_l"(ComponentPose=@n58, Alpha=@n54, Translation="0,0,0", Rotation="-58.290,0.009,-0.000", Scale="1,1,1") {
    rotationMode=BMM_Additive
    rotationSpace=BCS_BoneSpace
    alphaBoolBlend {blendOption=Linear}
    alphaScaleBiasClamp {bInterpResult=true}
    boneToModify {boneName=dyn_shortchain_f_l}
    rotation {pitch=-58.29, yaw=0.009}
   }
   n60 = "Transform (Modify) Bone - Bone: dyn_longchain_b_r"!(ComponentPose=@n59, Alpha:float=@n54, Translation="0,0,0", Rotation="-18.246,0.676,-0.077", Scale="1,1,1") -> (Pose) {
    rotationMode=BMM_Additive
    rotationSpace=BCS_BoneSpace
    alphaBoolBlend {blendOption=Linear}
    alphaScaleBiasClamp {bInterpResult=true}
    boneToModify {boneName=dyn_longchain_b_r}
    rotation=<rot -18.246 0.676 -0.077>
   }
   n61 = Get bIsInFrontEnd()
   n62 = "Transform (Modify) Bone - Bone: dyn_longchain_b_lower_l"!(ComponentPose=@n60, bAlphaBoolEnabled:bool=@n61, Translation="0,0,0", Rotation="17.681,1.469,2.610", Scale="1,1,1") -> (Pose) {
    alphaInputType=Bool
    rotationMode=BMM_Additive
    rotationSpace=BCS_BoneSpace
    alphaBoolBlend {blendOption=Linear}
    boneToModify {boneName=dyn_longchain_b_lower_l}
    rotation=<rot 17.681 1.469 2.61>
   }
   n63 = "Transform (Modify) Bone - Bone: dyn_longchain_f_lower_l"!(ComponentPose=@n62, bAlphaBoolEnabled:bool=@n61, Translation="0,0,0", Rotation="4.149,7.370,-19.203", Scale="1,1,1") -> (Pose) {
    alphaInputType=Bool
    rotationMode=BMM_Additive
    rotationSpace=BCS_BoneSpace
    alphaBoolBlend {blendOption=Linear}
    boneToModify {boneName=dyn_longchain_f_lower_l}
    rotation=<rot 4.149 7.37 -19.203>
   }
   n64 = "Transform (Modify) Bone - Bone: dyn_longchain_f_lower_r"!(ComponentPose=@n63, bAlphaBoolEnabled:bool=True, Translation="0,0,0", Rotation="-1.130,-15.325,59.480", Scale="1,1,1") -> (Pose) {
    alphaInputType=Bool
    rotationMode=BMM_Additive
    rotationSpace=BCS_BoneSpace
    alphaBoolBlend {blendOption=Linear}
    boneToModify {boneName=dyn_longchain_f_lower_r}
    rotation=<rot -1.13 -15.325 59.48>
   }
   n65 = "Transform (Modify) Bone - Bone: dyn_longchain_f_r"!(ComponentPose=@n64, bAlphaBoolEnabled:bool=True, Translation="0,0,0", Rotation="7.745,4.964,-26.628", Scale="1,1,1") -> (Pose) {
    alphaInputType=Bool
    rotationMode=BMM_Additive
    rotationSpace=BCS_BoneSpace
    alphaBoolBlend {blendOption=Linear}
    boneToModify {boneName=dyn_longchain_f_r}
    rotation=<rot 7.745 4.964 -26.628>
   }
   n66 = "Transform (Modify) Bone - Bone: dyn_longchain_b_lower_r"!(ComponentPose=@n65, bAlphaBoolEnabled:bool=True, Translation="0,0,0", Rotation="4.149,7.370,-19.203", Scale="1,1,1") -> (Pose) {
    alphaInputType=Bool
    rotationMode=BMM_Additive
    rotationSpace=BCS_BoneSpace
    alphaBoolBlend {blendOption=Linear}
    boneToModify {boneName=dyn_longchain_b_lower_r}
    rotation=<rot 4.149 7.37 -19.203>
   }
   n67 = "Transform (Modify) Bone - Bone: dyn_shortchain_buckle_r"!(ComponentPose=@n66, bAlphaBoolEnabled:bool=True, Translation="0,0,0", Rotation="-8.406,-14.074,0.034", Scale="1,1,1") -> (Pose) {
    alphaInputType=Bool
    rotationMode=BMM_Additive
    rotationSpace=BCS_BoneSpace
    alphaBoolBlend {blendOption=Linear}
    boneToModify {boneName=dyn_shortchain_buckle_r}
    rotation=<rot -8.406 -14.074 0.034>
   }
   n68 = Get bIsOnGrindRail()
   n69 = "Transform (Modify) Bone - Bone: dyn_shortchain_f_r"!(ComponentPose=@n67, bAlphaBoolEnabled:bool=@n68, Translation="0,0,0", Rotation="-26.350,-32.540,-26.115", Scale="1,1,1") -> (Pose) {
    alphaInputType=Bool
    rotationMode=BMM_Additive
    rotationSpace=BCS_BoneSpace
    alphaBoolBlend {blendOption=Linear}
    boneToModify {boneName=dyn_shortchain_f_r}
    rotation=<rot -26.3499 -32.54 -26.115>
   }
   n70 = "Transform (Modify) Bone - Bone: dyn_shortchain_f_l"!(ComponentPose=@n69, bAlphaBoolEnabled:bool=@n68, Translation="0,0,0", Rotation="51.544,-29.256,73.325", Scale="1,1,1") -> (Pose) {
    alphaInputType=Bool
    rotationMode=BMM_Additive
    rotationSpace=BCS_BoneSpace
    alphaBoolBlend {blendOption=Linear}
    boneToModify {boneName=dyn_shortchain_f_l}
    rotation=<rot 51.5439 -29.2559 73.325>
   }
   n71 = "Transform (Modify) Bone - Bone: dyn_longchain_f_lower_r"(ComponentPose=@n70, bAlphaBoolEnabled=@n68, Translation="0,0,0", Rotation="-41.839,9.078,4.238", Scale="1,1,1") {
    alphaInputType=Bool
    rotationMode=BMM_Additive
    rotationSpace=BCS_BoneSpace
    alphaBoolBlend {blendOption=Linear}
    boneToModify {boneName=dyn_longchain_f_lower_r}
    rotation=<rot -41.839 9.078 4.238>
   }
   n72 = "Transform (Modify) Bone - Bone: dyn_longchain_b_lower_r"(ComponentPose=@n71, bAlphaBoolEnabled=@n68, Translation="0,0,0", Rotation="-6.604,2.242,-19.717", Scale="1,1,1") {
    alphaInputType=Bool
    rotationMode=BMM_Additive
    rotationSpace=BCS_BoneSpace
    alphaBoolBlend {blendOption=Linear}
    boneToModify {boneName=dyn_longchain_b_lower_r}
    rotation=<rot -6.604 2.242 -19.717>
   }
   n73 = Get OnZipline() -> (OnZipline:bool)
   n74 = "Transform (Modify) Bone - Bone: dyn_longchain_f_lower_r"(ComponentPose=@n72, bAlphaBoolEnabled=@n73, Translation="0,0,0", Rotation="0,-40,0", Scale="1,1,1") {
    alphaInputType=Bool
    rotationMode=BMM_Additive
    rotationSpace=BCS_BoneSpace
    alphaBoolBlend {blendOption=Linear}
    boneToModify {boneName=dyn_longchain_f_lower_r}
    rotation {yaw=-40}
   }
   n75 = "Transform (Modify) Bone - Bone: dyn_longchain_f_lower_l"(ComponentPose=@n74, bAlphaBoolEnabled=@n73, Translation="0,0,0", Rotation="0,-50,0", Scale="1,1,1") {
    alphaInputType=Bool
    rotationMode=BMM_Additive
    rotationSpace=BCS_BoneSpace
    alphaBoolBlend {blendOption=Linear}
    boneToModify {boneName=dyn_longchain_f_lower_l}
    rotation {yaw=-50}
   }
   n76 = "Transform (Modify) Bone - Bone: dyn_longchain_b_lower_l"!(ComponentPose=@n75, Alpha:float=Get Isrelaxed() -> (Isrelaxed:double), Translation="0,0,0", Rotation="-15.650,0.002,-0.004", Scale="1,1,1") -> (Pose) {
    rotationMode=BMM_Additive
    rotationSpace=BCS_BoneSpace
    translationMode=BMM_Additive
    translationSpace=BCS_BoneSpace
    alphaBoolBlend {blendOption=Linear}
    alphaScaleBiasClamp {bInterpResult=true, interpSpeedDecreasing=5, interpSpeedIncreasing=5}
    boneToModify {boneName=dyn_longchain_b_lower_l}
    rotation=<rot -15.65 0.002 -0.004>
   }
   n77 = Get Alpha_IsTargeting() -> (Alpha_IsTargeting:float)
   n78 = "Transform (Modify) Bone - Bone: MASTER_sleeve_upperfront_r"(ComponentPose=@n76, Alpha=@n77, Translation="0,0,0", Rotation="15.562,37.481,15.988", Scale="1,1,1") {
    rotationMode=BMM_Additive
    rotationSpace=BCS_BoneSpace
    translationMode=BMM_Additive
    translationSpace=BCS_BoneSpace
    alphaBoolBlend {blendOption=Linear}
    alphaScaleBiasClamp {bInterpResult=true}
    boneToModify {boneName=MASTER_sleeve_upperfront_r}
    rotation=<rot 15.562 37.4809 15.988>
   }
   n79 = "Transform (Modify) Bone - Bone: MASTER_sleeve_upperbottom_r"(ComponentPose=@n78, Alpha=@n77, Translation="0,0,0", Rotation="34.311,-9.388,4.467", Scale="1,1,1") {
    rotationMode=BMM_Additive
    rotationSpace=BCS_BoneSpace
    translationMode=BMM_Additive
    translationSpace=BCS_BoneSpace
    alphaBoolBlend {blendOption=Linear}
    alphaScaleBiasClamp {bInterpResult=true}
    boneToModify {boneName=MASTER_sleeve_upperbottom_r}
    rotation=<rot 34.311 -9.388 4.467>
   }
   Output Pose(Result=Component To Local(ComponentPose=@n79)) {layerGroup=None}
   n80 = Control Rig(Source=Copy Pose From Mesh() {bUseAttachedParent=true}) {
    alphaBoolBlend {blendOption=Linear}
    controlRigAssetReference {blueprintRigClass=&/F_MED_RockerPunkAlt_ControlRig.F_MED_RockerPunkAlt_ControlRig_C}
   }
   Save cached pose 'Input node'(Pose=@n80) {cacheName=Input node}
   Get bIsKneeSliding()
  ]}
 EventGraph {nodes [
   Event BlueprintUpdateAnimation()
   > n1 = Sequence!(execute) -> (then_0, then_1, then_2:Exec, then_3:Exec, then_4:Exec, then_5:Exec, then_6:Exec, then_7:Exec, then_8:Exec, then_9:Exec)
   > n2 = Gravity Override State Function DEPRECATED(CharacterPartAnimInstance=Self-Reference(), GameTimeInSeconds=GetGameTimeInSeconds(), GravityOverrideParams=Get Gravity Override Params())
   > Set GravityOverride(GravityOverride=@n2)
   n3 = Branch(execute=@n1.then_1, Condition=OR Boolean(A=Get bIsFalling(), B=Get bIsJumping()))
   > Set Alpha(Alpha:double=Get Alpha_Jumping() -> (Alpha_Jumping:double)) -> (Output_Get:double)
   Set Alpha(execute=@n3.else, Alpha=Get Alpha_Reg() -> (Alpha_Reg:double))
   n4 = Get bIsKneeSliding()
   Set Crouchbody(execute=@n1.then_2, Crouchbody:bool=AND Boolean(A=Get bIsCrouching(), B=NOT Boolean(A=@n4))) -> (Output_Get:bool)
   n5 = Branch(execute=@n1.then_3, Condition=AND Boolean(A=Get bIsCrouching(), B=Get bIsAccelerating2D()))
   > Set CrouchMove fix(CrouchMove fix:double=Get Chain_Interp() -> (Chain_Interp:double)) -> (Output_Get:double)
   Set CrouchMove fix(execute=@n5.else, CrouchMove fix=Get Chain_Interp_0() -> (Chain_Interp_0:double))
   n6 = Get bIsAccelerating2D()
   n7 = Branch(execute=@n1.then_4, Condition=AND Boolean(A=Get bIsCrouching(), B=NOT Boolean(A=@n6)))
   > Set CrouchStill fix(CrouchStill fix:double=Get Chain_Interp()) -> (Output_Get:double)
   Set CrouchStill fix(execute=@n7.else, CrouchStill fix=Get Chain_Interp_0())
   n8 = AND Boolean(A=Get bIsJumping(), B=Get bIsMoving2D())
   n9 = Get bIsJumping()
   n10 = Get bIsFalling()
   n11 = Branch(
    execute=@n1.then_5
    Condition="Not Equal (Boolean)"(A=Get bIsSkyDiving(), B=OR Boolean!(A=@n10, B=@n9, C:bool=@n8) -> (ReturnValue)) -> (ReturnValue=~)
   )
   > Set Jump Fix(Jump Fix:double=Get Chain_Interp()) -> (Output_Get:double)
   Set Jump Fix(execute=@n11.else, Jump Fix=Get Chain_Interp_0())
   n12 = Branch(execute=@n1.then_6, Condition=Get bIsKneeSliding())
   > Set SlideFix(SlideFix:double=Get Chain_Interp()) -> (Output_Get:double)
   Set SlideFix(execute=@n12.else, SlideFix=Get Chain_Interp_0())
   n13 = Branch(execute=@n1.then_7, Condition="float > float"(A=Get AimPitch()))
   > Set Alpha_IsTargeting(Alpha_IsTargeting:float=Get positive() -> (positive:float)) -> (Output_Get:float)
   Set Alpha_IsTargeting(execute=@n13.else, Alpha_IsTargeting=Get negative() -> (negative:float))
   n14 = Branch(execute=@n1.then_8, Condition=NOT Boolean(A=Get bIsAccelerating2D()))
   > Set Isrelaxed(Isrelaxed:double=Get Chain_Interp()) -> (Output_Get:double)
   Set Isrelaxed(execute=@n14.else, Isrelaxed=Get Chain_Interp_0())
   n15 = Branch(execute=@n1.then_9, Condition=Get bIsSkyDiving())
   > Set sleves_alpha(sleves_alpha:double=0.75) -> (Output_Get:double)
   Set sleves_alpha(execute=@n15.else, sleves_alpha=1)
   n16 = Event BlueprintProcessAnimInputEvent()
   > n17 = Branch(Condition=IsNamed(self=@n16.AnimInputEvent, InEventName=aie.Player.Zipline.Entered))
   > Set OnZipline(OnZipline:bool=true) -> (Output_Get:bool)
   Branch(execute=@n17.else, Condition=IsNamed(self=@n16.AnimInputEvent, InEventName=aie.Player.Zipline.Exited))
   > Set OnZipline(OnZipline=false)
  ]}
]
variables [
 Alpha {type=number}
 Alpha_IsTargeting {type=number, default=1}
 Alpha_Jumping {type=number, default=0.6}
 Alpha_Reg {type=number, default=1}
 Chain_Interp {type=number, default=1}
 Chain_Interp_0 {type=number}
 Crouchbody {type=boolean}
 CrouchMove fix {type=number}
 CrouchStill fix {type=number}
 Gravity Override Params {
  type=GravityOverrideParamsStruct
  default {
   onGround_Moving {gravityOverride=<0 0 -1200>}
   skydive_FloatingDown {gravityOverride=<0 -900 2400>}
   skydive_GoingUp {gravityOverride=<-700 -700 -1500>}
   skydive_Parachuting {gravityOverride=<0 -800 -1500>}
   tactical_Sprint {gravityOverride=<0 -450 -980>}
   ziplining {gravityOverride=<0 0 -1200>}
  }
 }
 GravityOverride {type=Vector}
 Is Crouching fix {type=number}
 Isrelaxed {type=number}
 Jump Fix {type=number}
 negative {type=number}
 OnZipline {type=boolean}
 positive {type=number, default=1}
 sleves_alpha {type=number}
 SlideFix {type=number}
]
